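# Idle Niagara effect for the FeatherMudGlance pickaxe, with electricity, flames and a glow outline.
# Hilt and horn electricity meshes and flame meshes follow skeletal sockets chosen from index arrays; flames pass a smoothed position to their material, apparently for lag.
# Glow comes from an outline mesh plus a low variant of sprites at hand-placed points tracing the weapon.
# System Timeline time ramps electricity spawn rates and flame and glow alpha.
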NiagaraSystem /BRCosmetics/Weapons/FORT_Melee/Pickaxe_FeatherMudGlance/FX/NS_Pickaxe_FeatherMudGlance_Idle
dependencies {
 usedDataInterfaces [
  &NiagaraDataInterfaceCurve
  &NiagaraDataInterfaceArrayFloat3
  &NiagaraDataInterfaceArrayInt32
  &NiagaraDataInterfaceSkeletalMesh
 ]
 usedDynamicInputs [&Boolean_OrOperation]
 usedModules [
  &self:SystemSpawnScript.SystemScriptSource.SystemScriptGraph.NiagaraNodeAssignment_0.SetVariables_#1
  &SystemState
  &self:SystemSpawnScript.SystemScriptSource.SystemScriptGraph.NiagaraNodeAssignment_1.SetVariables_#2
  &Timeline
 ]
}
emitters {
 Em_Electricity_Hilt {
  emitterUpdateScript [
   &EmitterState(Life Cycle Mode=::NewEnumerator1, Loop Duration=1, UseLoopDelay=false)
   &SpawnRate(
    Spawn Probability=0.8
    SpawnRate=&FloatFromCurve(FloatCurve=DataInterface() {propertyValues {Curve {keys [<at 0 0>,<at 1 1>]}}}, CurveIndex=$System.Timeline.Time, Scale Curve=9)
    Use Spawn Probability=false
   )
  ]
  particleSpawnScript [
   &PartitionParticles()
   &InitialMeshOrientation(
    Mesh Axis To Orient=&SelectVectorFromArray(
     Direct Array Index=$Particles.PartitionParticles.Partition
     Vector Selection Array=DataInterface() {propertyValues {FloatData [<1 0 0>, <1 0 0>, <0 0 1>, <0 0 1>]}}
    )
    Mesh Orientation Mode=::NewEnumerator2
    Orientation Vector=&SelectVectorFromArray(
     Direct Array Index=$Particles.PartitionParticles.Partition
     Vector Selection Array=DataInterface() {propertyValues {FloatData [<0 0 0>, <0 0 0>, <0 0 -1>, <0 0 1>]}}
    )
    Rotation=&RandomRangeVector(Minimum=<-0.05 -0.05 -1>, Maximum=<0.05 0.05 1>)
    Use Rotation Vector=true
   )
   &InitializeParticle(
    Lifetime Max=0.3
    Lifetime Min=0.2
    Lifetime Mode=::NewEnumerator1
    Mesh Scale=&SelectVectorFromArray(
     Direct Array Index=$Particles.PartitionParticles.Partition
     Vector Selection Array=DataInterface() {propertyValues {FloatData [<0.2 0.2 0.5>, <0.2 0.2 0.5>, <0.2 0.2 0.2>, <0.2 0.2 0.2>]}}
    )
    Mesh Scale Mode=::NewEnumerator4
    UsePositionOffset=true
    Write Material Random=true
   )
  ]
  particleUpdateScript [
   &ParticleState()
   &DynamicMaterialParameters(
    Index 0 Param 1=&MakeFloatFromVector(Vector=$Particles.Scale, Channel=::NewEnumerator0)
    Index 0 Param 2=&RandomRangeFloat(Minimum=0, Maximum=10)
    Index 0 Param 3=&FloatFromCurve(
     FloatCurve=DataInterface() {propertyValues {Curve {keys [{interpMode=RCIM_Cubic, leaveTangent=0.5585, tangentMode=RCTM_Break, time=0.8, value=2}, {tangentMode=RCTM_User, time=1, value=0}]}}}
     Scale Curve=0.02
    )
    Index 0 Param 4=0.5
    Index 1 Param 1=&FloatFromCurve(FloatCurve=DataInterface() {propertyValues {Curve {keys [
         {interpMode=RCIM_Cubic, leaveTangent=1.1034, tangentMode=RCTM_Break, time=0, value=0}
         {arriveTangent=2.2399, tangentMode=RCTM_User, time=0.5, value=1}
        ]}}})
    Index 1 Param 2=&FloatFromCurve(FloatCurve=DataInterface() {propertyValues {Curve {keys [<at 0.5 0>, {arriveTangent=3.2053, tangentMode=RCTM_User, time=0.9, value=0}]}}})
    Index 1 Param 3=&FloatFromCurve(FloatCurve=DataInterface() {propertyValues {Curve {keys [<at 0.5 0>,<at 1 1>]}}})
    Index 1 Param 4=&FloatFromCurve(
     FloatCurve=DataInterface() {propertyValues {Curve {keys [
         {interpMode=RCIM_Cubic, leaveTangent=0.1169, tangentMode=RCTM_Break, time=0, value=0}
         {arriveTangent=2.5708, interpMode=RCIM_Cubic, leaveTangent=2.5708, tangentMode=RCTM_User, time=1, value=1}
        ]}}}
     Scale Curve=&RandomRangeFloat(Minimum=40, Maximum=50)
    )
    Index 2 Param 1=&FloatFromCurve(
     FloatCurve=DataInterface() {propertyValues {Curve {keys [
         {interpMode=RCIM_Cubic, leaveTangent=-0.0025, tangentMode=RCTM_Break, time=0, value=0.5}
         {arriveTangent=-0.0316, interpMode=RCIM_Cubic, tangentMode=RCTM_Break, time=1, value=-0.2}
        ]}}}
     Scale Curve=0.5
    )
    Index 2 Param 2=0.5
    Module.Index0Param0Write=true
    Module.Index0Param1Write=true
    Module.Index0Param2Write=true
    Module.Index0Param3Write=true
    Module.Index1Param0Write=true
    Module.Index1Param1Write=true
    Module.Index1Param2Write=true
    Module.Index1Param3Write=true
    Module.Index2Param0Write=true
    Module.Index2Param1Write=true
    Module.Index2Param2Write=true
    Module.Index2Param3Write=true
    Module.Index3Param0Write=true
    Module.Index3Param1Write=true
    Module.Index3Param2Write=true
    Module.Index3Param3Write=true
    Write Parameter Index 1=true
    Write Parameter Index 2=true
    Write Parameter Index 3=true
   )
   &SkeletalMeshLocation(
    Bone / Socket Index=&SelectIntFromArray(Direct Array Index=$Particles.PartitionParticles.Partition, Int Selection Array=DataInterface() {propertyValues {IntData [1,2]}})
    MaskSpawnGroup=false
    Mesh Sampling Type=::NewEnumerator5
    Sampled Orientation Coordinate Space=::World
    Sampled Position Offset=&SelectVectorFromArray(Direct Array Index=$Particles.PartitionParticles.Partition, Vector Selection Array=DataInterface() {propertyValues {FloatData []}})
    Skeletal Mesh=$System.Skeletal Mesh
    Socket Sampling Mode=::NewEnumerator5
   )
   &ScaleColor(
    Scale RGB=&Multiply_VectorByFloat(
     Vector=<1 1 1>
     Float=&FloatFromCurve(FloatCurve=DataInterface() {propertyValues {Curve {keys [<at 0 2>,<at 0.05 1>,<at 1 1>]}}}, Scale Curve=8)
    )
    ScaleA=true
    ScaleRGB=true
   )
  ]
  rendererClasses [&NiagaraMeshRendererProperties]
  renderers [
   {rendererClass=&NiagaraMeshRendererProperties, rendererIndex=0}
   {rendererClass=&NiagaraMeshRendererProperties, rendererIndex=1}
  ]
 }
 Em_Electricity_Horns {
  emitterUpdateScript [
   &EmitterState(Life Cycle Mode=::NewEnumerator1, Loop Duration=1, UseLoopDelay=false)
   &SpawnRate(
    Spawn Probability=0.8
    SpawnRate=&FloatFromCurve(FloatCurve=DataInterface() {propertyValues {Curve {keys [<at 0 0>,<at 1 1>]}}}, CurveIndex=$System.Timeline.Time, Scale Curve=18)
    Use Spawn Probability=false
   )
  ]
  particleSpawnScript [
   &PartitionParticles()
   &InitialMeshOrientation(Rotation=&RandomRangeVector(Minimum=<0 0 0>, Maximum=<1 1 1>), Use Rotation Vector=true)
   &InitializeParticle(
    Lifetime Max=0.3
    Lifetime Min=0.2
    Lifetime Mode=::NewEnumerator1
    Mesh Scale=<0.5 0.5 0.5>
    Mesh Scale Mode=::NewEnumerator4
    UsePositionOffset=true
    Write Material Random=true
   )
  ]
  particleUpdateScript [
   &ParticleState()
   &DynamicMaterialParameters(
    Index 0 Param 1=&MakeFloatFromVector(Vector=$Particles.Scale, Channel=::NewEnumerator0)
    Index 0 Param 2=&RandomRangeFloat(Minimum=0, Maximum=10)
    Index 0 Param 3=&FloatFromCurve(
     FloatCurve=DataInterface() {propertyValues {Curve {keys [{interpMode=RCIM_Cubic, leaveTangent=0.5585, tangentMode=RCTM_Break, time=0.8, value=2}, {tangentMode=RCTM_User, time=1, value=0}]}}}
     Scale Curve=0.02
    )
    Index 0 Param 4=0.5
    Index 1 Param 1=&FloatFromCurve(FloatCurve=DataInterface() {propertyValues {Curve {keys [
         {interpMode=RCIM_Cubic, leaveTangent=1.1034, tangentMode=RCTM_Break, time=0, value=0}
         {arriveTangent=2.2399, tangentMode=RCTM_User, time=0.5, value=1}
        ]}}})
    Index 1 Param 2=&FloatFromCurve(FloatCurve=DataInterface() {propertyValues {Curve {keys [<at 0.5 0>, {arriveTangent=3.2053, tangentMode=RCTM_User, time=0.9, value=0}]}}})
    Index 1 Param 3=&FloatFromCurve(FloatCurve=DataInterface() {propertyValues {Curve {keys [<at 0.5 0>,<at 1 1>]}}})
    Index 1 Param 4=&FloatFromCurve(
     FloatCurve=DataInterface() {propertyValues {Curve {keys [
         {interpMode=RCIM_Cubic, leaveTangent=0.1169, tangentMode=RCTM_Break, time=0, value=0}
         {arriveTangent=2.5708, interpMode=RCIM_Cubic, leaveTangent=2.5708, tangentMode=RCTM_User, time=1, value=1}
        ]}}}
     Scale Curve=&RandomRangeFloat(Minimum=40, Maximum=50)
    )
    Index 2 Param 1=&FloatFromCurve(
     FloatCurve=DataInterface() {propertyValues {Curve {keys [
         {interpMode=RCIM_Cubic, leaveTangent=-0.0025, tangentMode=RCTM_Break, time=0, value=0.5}
         {arriveTangent=-0.0316, interpMode=RCIM_Cubic, tangentMode=RCTM_Break, time=1, value=-0.2}
        ]}}}
     Scale Curve=0.5
    )
    Index 2 Param 2=0.5
    Module.Index0Param0Write=true
    Module.Index0Param1Write=true
    Module.Index0Param2Write=true
    Module.Index0Param3Write=true
    Module.Index1Param0Write=true
    Module.Index1Param1Write=true
    Module.Index1Param2Write=true
    Module.Index1Param3Write=true
    Module.Index2Param0Write=true
    Module.Index2Param1Write=true
    Module.Index2Param2Write=true
    Module.Index2Param3Write=true
    Module.Index3Param0Write=true
    Module.Index3Param1Write=true
    Module.Index3Param2Write=true
    Module.Index3Param3Write=true
    Write Parameter Index 1=true
    Write Parameter Index 2=true
    Write Parameter Index 3=true
   )
   &SkeletalMeshLocation(
    Bone / Socket Index=&SelectIntFromArray(Direct Array Index=$Particles.PartitionParticles.Partition, Int Selection Array=DataInterface() {propertyValues {IntData [3,4]}})
    MaskSpawnGroup=false
    Mesh Sampling Type=::NewEnumerator5
    Sampled Orientation Coordinate Space=::World
    Sampled Position Offset=&SelectVectorFromArray(Direct Array Index=$Particles.PartitionParticles.Partition, Vector Selection Array=DataInterface() {propertyValues {FloatData []}})
    Skeletal Mesh=$System.Skeletal Mesh
    Socket Sampling Mode=::NewEnumerator5
   )
   &ScaleColor(
    Scale RGB=&Multiply_VectorByFloat(
     Vector=<1 1 1>
     Float=&FloatFromCurve(FloatCurve=DataInterface() {propertyValues {Curve {keys [<at 0 2>,<at 0.05 1>,<at 1 1>]}}}, Scale Curve=8)
    )
    ScaleA=true
    ScaleRGB=true
   )
  ]
  rendererClasses [&NiagaraMeshRendererProperties]
  renderers [
   {rendererClass=&NiagaraMeshRendererProperties, rendererIndex=0}
   {rendererClass=&NiagaraMeshRendererProperties, rendererIndex=1}
  ]
 }
 Em_Flames {
  emitterUpdateScript [
   &EmitterState(Life Cycle Mode=::NewEnumerator1, Loop Behavior=::NewEnumerator1, Loop Duration Mode=::NewEnumerator1, UseLoopDelay=false)
   &SpawnBurst_Instantaneous(Spawn Count=7, Use Spawn Probability=false, UseLoopCountLimit=false)
  ]
  particleSpawnScript [
   &InitializeParticle(
    Color=<rgba 3.3978 0.3853 5 1>
    Lifetime=5
    Mesh Scale=&SelectVectorFromArray(
     Direct Array Index=$Engine.ExecIndex
     Vector Selection Array=DataInterface() {propertyValues {FloatData [<1 1 6>, <1 1 8>, <1 1 6>, <1 1 5>, <1 1 5>]}}
    )
    Mesh Scale Mode=::NewEnumerator4
    UsePositionOffset=true
    Write Material Random=true
   )
   &InitialMeshOrientation(Mesh Orientation Mode=::NewEnumerator1, Use Rotation Vector=true)
  ]
  particleUpdateScript [
   &ParticleState(Kill Particles When Lifetime Has Elapsed=false, Let Infinitely Lived Particles Die When Emitter Deactivates=true)
   &SkeletalMeshLocation(
    Bone / Socket Index=&SelectIntFromArray(
     Direct Array Index=&ReturnExecIndex()
     Int Selection Array=DataInterface() {propertyValues {IntData [0,0,0,3,4]}}
    )
    MaskSpawnGroup=false
    Mesh Sampling Type=::NewEnumerator5
    Sampled Position Offset=&SelectVectorFromArray(
     Direct Array Index=&ReturnExecIndex()
     Vector Selection Array=DataInterface() {propertyValues {FloatData [<8 0 0>, <0 0 0>, <-8 0 0>, <2 0 0>, <-2 0 0>]}}
    )
    Skeletal Mesh=$System.Skeletal Mesh
    Socket Sampling Mode=::NewEnumerator5
   )
   &InterpolateOverTime(Rate Of Change=20, Target Vector Value:NiagaraPosition=$Particles.Position, Variable Type=::NewEnumerator2)
   &DynamicMaterialParameters(
    Index 0 Param 1=15
    Index 0 Param 2=40
    Index 0 Param 3=2
    Index 0 Param 4=1
    Index 1 Param 1=&MakeFloatFromVector(Vector=$Particles.InterpolateOverTime.InterpolateOverTime_V3_.Moving Average, Channel=::NewEnumerator0)
    Index 1 Param 2=&MakeFloatFromVector(Vector=$Particles.InterpolateOverTime.InterpolateOverTime_V3_.Moving Average, Channel=::NewEnumerator1)
    Index 1 Param 3=&MakeFloatFromVector(Vector=$Particles.InterpolateOverTime.InterpolateOverTime_V3_.Moving Average, Channel=::NewEnumerator2)
    Index 2 Param 1=&RandomRangeFloat(Minimum=0, Maximum=15)
    Module.Index0Param0Write=true
    Module.Index0Param1Write=true
    Module.Index0Param2Write=true
    Module.Index0Param3Write=true
    Module.Index1Param0Write=true
    Module.Index1Param1Write=true
    Module.Index1Param2Write=true
    Module.Index1Param3Write=true
    Module.Index2Param0Write=true
    Module.Index2Param1Write=true
    Module.Index2Param2Write=true
    Module.Index2Param3Write=true
    Write Parameter Index 1=true
    Write Parameter Index 2=true
   )
   &ScaleColor(
    Scale Alpha=&FloatFromCurve(FloatCurve=DataInterface() {propertyValues {Curve {keys [<at 0 0>,<at 1 1>]}}}, CurveIndex=$System.Timeline.Time)
    ScaleA=true
    ScaleRGB=true
   )
  ]
  rendererClasses [&NiagaraMeshRendererProperties]
  renderers [
   {rendererClass=&NiagaraMeshRendererProperties, rendererIndex=0}
   {rendererClass=&NiagaraMeshRendererProperties, rendererIndex=1}
  ]
 }
 Em_Glow_Outline {
  emitterUpdateScript [
   &EmitterState(Life Cycle Mode=::NewEnumerator1, Loop Behavior=::NewEnumerator1, Loop Duration Mode=::NewEnumerator1, UseLoopDelay=false)
   &SpawnBurst_Instantaneous(Use Spawn Probability=false, UseLoopCountLimit=false)
  ]
  particleSpawnScript [&InitializeParticle(
    Mesh Scale Mode=::NewEnumerator3
    Sprite Size Mode=::NewEnumerator3
    Uniform Sprite Size=1
    UsePositionOffset=false
    Write Material Random=true
   )]
  particleUpdateScript [
   &ParticleState(Kill Particles When Lifetime Has Elapsed=false, Let Infinitely Lived Particles Die When Emitter Deactivates=true)
   &ScaleColor(
    Scale Alpha=&FloatFromCurve(FloatCurve=DataInterface() {propertyValues {Curve {keys [<at 0 0>,<at 1 1>]}}}, CurveIndex=$System.Timeline.Time)
    ScaleA=true
    ScaleRGB=true
   )
  ]
  rendererClasses [&NiagaraMeshRendererProperties]
  renderers [{rendererClass=&NiagaraMeshRendererProperties, rendererIndex=0}]
 }
 Em_Glow_Outline_Low {
  emitterUpdateScript [
   &EmitterState(Life Cycle Mode=::NewEnumerator1, Loop Behavior=::NewEnumerator1, Loop Duration Mode=::NewEnumerator1, UseLoopDelay=false)
   &SpawnBurst_Instantaneous(Spawn Count=64, Use Spawn Probability=false, UseLoopCountLimit=false)
  ]
  particleSpawnScript [
   &InitializeParticle(
    Color=<rgba 0.9684 0.1837 1 1>
    Mesh Scale Mode=::NewEnumerator3
    Position=&SelectVectorFromArray(
     Direct Array Index=&ReturnExecIndex()
     Vector Selection Array=DataInterface() {propertyValues {FloatData [
        <-19.9298 9.924 37.4759>
        <-20.802 6.9306 43.9881>
        <-21.6743 3.9373 50.5002>
        <-22.5465 0.944 57.0124>
        <21.4017 9.9903 37.4313>
        <22.3786 6.9763 43.8588>
        <23.3555 3.9624 50.2863>
        <24.3324 0.9484 56.7138>
        <19.9162 10.5542 55.6318>
        <13.314 9.8757 55.8186>
        <6.7119 9.1972 56.0053>
        <0.1097 8.5187 56.1921>
        <-28.8683 -4.9623 60.4336>
        <-26.761 -2.9936 59.2932>
        <-24.6538 -1.0248 58.1528>
        <26.5029 -1.145 57.9096>
        <28.6734 -3.2385 59.1054>
        <30.8439 -5.332 60.3013>
        <-16.9586 11.042 55.46>
        <-11.2691 10.2009 55.704>
        <-5.5797 9.3598 55.948>
        <-11.8832 4.2363 61.6311>
        <-14.609 2.84 63.6451>
        <-17.3348 1.4438 65.6592>
        <-20.0606 0.0476 67.6732>
        <15.3402 3.4555 61.651>
        <17.4033 2.3556 63.6598>
        <19.4665 1.2556 65.6686>
        <21.5297 0.1557 67.6773>
        <29.629 0.1192 52.5788>
        <27.8635 0.3956 53.9571>
        <26.0979 0.672 55.3355>
        <-27.9261 0.2019 52.7621>
        <-26.1329 0.4493 54.1789>
        <-24.3397 0.6966 55.5956>
        <21.335 1.7841 58.3595>
        <18.3376 2.6198 60.0053>
        <-18.9921 2.0414 58.5519>
        <-15.4377 3.1388 60.0915>
        <0.7477 20.9924 41.1221>
        <0.5351 16.8345 46.1455>
        <0.3224 12.6766 51.1688>
        <10.2634 5.1433 59.8314>
        <5.1866 6.831 58.0117>
        <-7.8856 5.6638 59.8181>
        <-3.8879 7.0912 58.0051>
        <0.7923 7.1974 -72.6525>
        <0.7496 7.28 -64.5997>
        <0.7069 7.3626 -56.5469>
        <0.6643 7.4452 -48.4941>
        <0.6216 7.5278 -40.4413>
        <0.579 7.6103 -32.3886>
        <0.5363 7.6929 -24.3358>
        <0.4937 7.7755 -16.283>
        <0.451 7.8581 -8.2302>
        <0.4083 7.9407 -0.1774>
        <0.3657 8.0232 7.8754>
        <0.323 8.1058 15.9281>
        <0.2804 8.1884 23.9809>
        <0.2377 8.271 32.0337>
        <0.1951 8.3536 40.0865>
        <0.1524 8.4361 48.1393>
        <0 2.2325 64.1>
       ]}}
    )
    Position Mode=::NewEnumerator1
    Sprite Size=<20 25>
    Sprite Size Mode=::NewEnumerator4
    UsePositionOffset=false
    Write Material Random=true
   )
   /SetVariables(Particles.Velocity=&SelectVectorFromArray(
     Direct Array Index=&ReturnExecIndex()
     Vector Selection Array=DataInterface() {propertyValues {FloatData [
        <-0.1208 -0.4146 0.902>
        <-0.1208 -0.4146 0.902>
        <-0.1208 -0.4146 0.902>
        <0.861 0.4997 -0.0947>
        <0.1363 -0.4206 0.8969>
        <0.1363 -0.4206 0.8969>
        <0.1363 -0.4206 0.8969>
        <-0.8614 0.5035 -0.0662>
        <-0.9944 -0.1022 0.0281>
        <-0.9944 -0.1022 0.0281>
        <-0.9944 -0.1022 0.0281>
        <-0.4603 0.445 -0.7682>
        <0.6795 0.6349 -0.3677>
        <0.6795 0.6349 -0.3677>
        <0.6795 0.6349 -0.3677>
        <0.6691 -0.6453 0.3686>
        <0.6691 -0.6453 0.3686>
        <0.6691 -0.6453 0.3686>
        <0.9884 -0.1461 0.0424>
        <0.9884 -0.1461 0.0424>
        <0.9884 -0.1461 0.0424>
        <0.913 0.3526 -0.2051>
        <-0.7436 -0.3809 0.5495>
        <-0.7436 -0.3809 0.5495>
        <-0.7436 -0.3809 0.5495>
        <-0.9206 0.3398 -0.1926>
        <0.6693 -0.3568 0.6517>
        <0.6693 -0.3568 0.6517>
        <0.6693 -0.3568 0.6517>
        <-0.7823 0.1225 0.6107>
        <-0.7823 0.1225 0.6107>
        <-0.7823 0.1225 0.6107>
        <0.7801 0.1076 0.6163>
        <0.7801 0.1076 0.6163>
        <0.7801 0.1076 0.6163>
        <-0.8515 0.2374 0.4675>
        <-0.8515 0.2374 0.4675>
        <0.8829 0.2726 0.3824>
        <0.8829 0.2726 0.3824>
        <-0.0326 -0.6373 0.7699>
        <-0.0326 -0.6373 0.7699>
        <-0.0326 -0.6373 0.7699>
        <-0.8984 0.2987 -0.322>
        <-0.8984 0.2987 -0.322>
        <0.8661 0.3093 -0.3928>
        <0.8661 0.3093 -0.3928>
        <-0.0053 0.0103 0.9999>
        <-0.0053 0.0103 0.9999>
        <-0.0053 0.0103 0.9999>
        <-0.0053 0.0103 0.9999>
        <-0.0053 0.0103 0.9999>
        <-0.0053 0.0103 0.9999>
        <-0.0053 0.0103 0.9999>
        <-0.0053 0.0103 0.9999>
        <-0.0053 0.0103 0.9999>
        <-0.0053 0.0103 0.9999>
        <-0.0053 0.0103 0.9999>
        <-0.0053 0.0103 0.9999>
        <-0.0053 0.0103 0.9999>
        <-0.0053 0.0103 0.9999>
        <-0.0053 0.0103 0.9999>
        <-0.0053 0.0103 0.9999>
        <0.0109 0.6222 -0.7828>
       ]}}
    )) {isSetParameters=true}
  ]
  particleUpdateScript [
   &ParticleState(Kill Particles When Lifetime Has Elapsed=false, Let Infinitely Lived Particles Die When Emitter Deactivates=true)
   &ScaleColor(
    Scale Alpha=&FloatFromCurve(FloatCurve=DataInterface() {propertyValues {Curve {keys [<at 0 0>,<at 1 1>]}}}, CurveIndex=$System.Timeline.Time)
    ScaleA=true
    ScaleRGB=true
   )
  ]
  rendererClasses [&NiagaraSpriteRendererProperties]
  renderers [{rendererClass=&NiagaraSpriteRendererProperties}]
 }
}
summary {
 systemName=NS_Pickaxe_FeatherMudGlance_Idle
 emitters [
  {emitterName=Em_Glow_Outline, rendererClasses [&NiagaraMeshRendererProperties]}
  {emitterName=Em_Electricity_Hilt, rendererClasses [&NiagaraMeshRendererProperties]}
  {emitterName=Em_Flames, rendererClasses [&NiagaraMeshRendererProperties]}
  {emitterName=Em_Electricity_Horns, rendererClasses [&NiagaraMeshRendererProperties]}
  {emitterName=Em_Glow_Outline_Low, rendererClasses [&NiagaraSpriteRendererProperties]}
 ]
 userVariables [
  User.bIsFrontEnd {type=NiagaraBool}
  User.bIsFrontEndPreview {type=NiagaraBool}
  User.bWeakSpotHit {type=NiagaraBool}
 ]
}
