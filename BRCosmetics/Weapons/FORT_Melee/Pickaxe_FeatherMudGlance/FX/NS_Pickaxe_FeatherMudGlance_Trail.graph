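# Niagara system for the FeatherMudGlance pickaxe's melee trail, drawn by a single ribbon emitter.
# FN anim-trail modules spawn, initialize and update short-lived particles from bone data on the system's skeletal mesh, with a ribbon ID offset utility at spawn.
# Ribbon facing follows each particle's initial velocity.
# Age curves drive opacity erosion and alpha fade, while time-of-day scalars make the color brighter by day and dimmer at night.
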NiagaraSystem /BRCosmetics/Weapons/FORT_Melee/Pickaxe_FeatherMudGlance/FX/NS_Pickaxe_FeatherMudGlance_Trail
dependencies {
 usedDataInterfaces [&NiagaraDataInterfaceCurve, &NiagaraDataInterfaceSkeletalMesh]
 usedDynamicInputs []
 usedModules [
  &self:SystemSpawnScript.SystemScriptSource.SystemScriptGraph.NiagaraNodeAssignment_0.SetVariables_#1
  &SystemState
 ]
}
emitters {Em_AnimTrail {
  emitterSpawnScript [&FN_RibbonIDOffsetUtility()]
  emitterUpdateScript [
   &EmitterState()
   &FN_SpawnAnimTrail()
  ]
  particleSpawnScript [
   &FN_InitializeAnimTrailBoneIndecies()
   &FN_InitializeAnimTrail(AnimTrailWidth=4, Lifetime=0.4, SkeletalMesh=$System.Skeletal Mesh)
   &AddVelocity(Velocity=<0 0 1>)
  ]
  particleUpdateScript [
   &ParticleState()
   /SetVariables(Particles.RibbonFacing=$Particles.Initial.Velocity) {isSetParameters=true}
   &FN_UpdateAnimTrail(SkeletalMesh=$System.Skeletal Mesh)
   &FN_RibbonMaterial(
    Opacity Erosion:NiagaraFloat=&FloatFromCurve(FloatCurve=DataInterface() {propertyValues {Curve {keys [{interpMode=RCIM_Cubic, tangentMode=RCTM_User, time=0, value=0}, {interpMode=RCIM_Cubic, tangentMode=RCTM_User, time=1, value=1}]}}})
    Scale Alpha:NiagaraFloat=&FloatFromCurve(FloatCurve=DataInterface() {propertyValues {Curve {keys [
         {interpMode=RCIM_Cubic, tangentMode=RCTM_User, time=0, value=1}
         {arriveTangent=-3.2703, interpMode=RCIM_Cubic, leaveTangent=-3.2703, tangentMode=RCTM_User, time=1, value=0}
        ]}}})
   )
   &ScaleColor(
    Scale Alpha=&Multiply_Float(
     A=&FNTimeOfDayScalar(Night Scale=1, Morning Scale=1, Evening Scale=1)
     B=&FloatFromCurve(FloatCurve=DataInterface() {propertyValues {Curve {keys [<at 0 1 RCIM_Cubic>, {arriveTangent=-4.1378, tangentMode=RCTM_User, time=0.8, value=0}]}}})
    )
    Scale RGB=&Multiply_VectorByFloat(
     Vector=<1 1 1>
     Float=&FNTimeOfDayScalar(Night Scale=0.75, Morning Scale=1.2, Day Scale=1.2, Evening Scale=0.75)
    )
    ScaleA=true
    ScaleRGB=true
   )
   &SolveForcesAndVelocity(Clamp Velocity=false, Limit Acceleration=false)
  ]
  rendererClasses [&NiagaraRibbonRendererProperties]
  renderers [{rendererClass=&NiagaraRibbonRendererProperties}]
 }}
summary {
 systemName=NS_Pickaxe_FeatherMudGlance_Trail
 emitters [{emitterName=Em_AnimTrail, rendererClasses [&NiagaraRibbonRendererProperties]}]
 userVariables [User.RibbonIDOffset {type=NiagaraInt32}]
}
