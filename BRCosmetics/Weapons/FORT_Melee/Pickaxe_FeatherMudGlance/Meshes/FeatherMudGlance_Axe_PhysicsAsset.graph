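# Physics asset for the FeatherMudGlance pickaxe axe mesh, a two-handed melee cosmetic weapon.
# It holds a single body on the Melee_TwoHanded_Handle bone, shaped as one rotated capsule offset along the handle.
# With no constraints, the weapon collides as one rigid piece with no jointed or simulated secondary parts.
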PhysicsAsset /BRCosmetics/Weapons/FORT_Melee/Pickaxe_FeatherMudGlance/Meshes/FeatherMudGlance_Axe_PhysicsAsset
bodies {Melee_TwoHanded_Handle [{shapeType=Capsule, length=73.0124, radius=32.453, center=<3.149 -1.0138 26.4128>, rotation=<rot 0 -90 -0.0329>}]}
constraints []
